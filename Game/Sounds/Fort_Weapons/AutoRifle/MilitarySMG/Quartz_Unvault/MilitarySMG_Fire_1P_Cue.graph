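# First-person fire sound cue for the Quartz Unvault Military SMG.
# Randomized close shots blend with quieter Machine Pistol mechanical and shell-impact layers, with volume falling as the TimeSinceStarted parameter rises.
# A parallel LowAmmo branch adds a Machine Pistol warning sound only when that parameter is true.
# Separate sound class overrides route the fire and low-ammo layers, and the random nodes are excluded from branch culling.
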SoundCue /Game/Sounds/Fort_Weapons/AutoRifle/MilitarySMG/Quartz_Unvault/MilitarySMG_Fire_1P_Cue
root=Mixer_10
nodes [
 {id=Branch_11, children [Modulator_21,null,null], properties {boolParameterName=LowAmmo}}
 {id=Mixer_10, children [SoundClass_6,SoundClass_7], properties {inputVolume [0.8,1]}}
 {id=Mixer_11, children [ModulatorContinuous_6], properties {inputVolume [1]}}
 {id=Mixer_9, children [Random_10,Random_11,Modulator_22], properties {inputVolume [1,0.3,0.3]}}
 {
  id=ModulatorContinuous_6
  children [Modulator_18]
  properties {volumeModulationParams {default=0, maxOutput=0.6, minOutput=1, parameterName=TimeSinceStarted}}
 }
 {id=Modulator_17, children [Mixer_9], properties {pitchMax=1, pitchMin=1}}
 {id=Modulator_18, children [Modulator_17], properties {pitchMax=1, pitchMin=1}}
 {id=Modulator_19, children [Branch_11], properties {pitchMax=1, pitchMin=1, volumeMax=1.15, volumeMin=1}}
 {id=Modulator_21, children [WavePlayer_74], properties {pitchMax=1, pitchMin=1, volumeMax=1, volumeMin=1}}
 {id=Modulator_22, children [Random_13]}
 {
  id=Random_10
  children [WavePlayer_54, WavePlayer_55, WavePlayer_56, WavePlayer_57, WavePlayer_58, WavePlayer_59, WavePlayer_60, WavePlayer_61]
  properties {bShouldExcludeFromBranchCulling=true, weights [1,1,1,1,1,1,1,1]}
 }
 {
  id=Random_11
  children [WavePlayer_66, WavePlayer_67, WavePlayer_68, WavePlayer_69, WavePlayer_70, WavePlayer_71]
  properties {bShouldExcludeFromBranchCulling=true, weights [1,1,1,1,1,1]}
 }
 {
  id=Random_13
  children [WavePlayer_75, WavePlayer_76, WavePlayer_77, WavePlayer_78, WavePlayer_79, WavePlayer_80]
  properties {bShouldExcludeFromBranchCulling=true, weights [1,1,1,1,1,1]}
 }
 {
  id=SoundClass_6
  children [Modulator_19]
  properties {soundClassOverride=/Game/Sounds/FortSoundClassesAndModes/Fort_WeaponRanged_LowAmmo_Critical_PSM}
 }
 {id=SoundClass_7, children [Mixer_11], properties {soundClassOverride=&Fort_WeaponRanged_Fire_1P_PSM}}
 {
  id=WavePlayer_54
  properties {soundWaveAssetPtr=&/Military_SMG_Quartz_Fire_Close_01}
 }
 {
  id=WavePlayer_55
  properties {soundWaveAssetPtr=&/Military_SMG_Quartz_Fire_Close_02}
 }
 {
  id=WavePlayer_56
  properties {soundWaveAssetPtr=&/Military_SMG_Quartz_Fire_Close_03}
 }
 {
  id=WavePlayer_57
  properties {soundWaveAssetPtr=&/Military_SMG_Quartz_Fire_Close_04}
 }
 {
  id=WavePlayer_58
  properties {soundWaveAssetPtr=&/Military_SMG_Quartz_Fire_Close_05}
 }
 {
  id=WavePlayer_59
  properties {soundWaveAssetPtr=&/Military_SMG_Quartz_Fire_Close_06}
 }
 {
  id=WavePlayer_60
  properties {soundWaveAssetPtr=&/Military_SMG_Quartz_Fire_Close_07}
 }
 {
  id=WavePlayer_61
  properties {soundWaveAssetPtr=&/Military_SMG_Quartz_Fire_Close_08}
 }
 {id=WavePlayer_66, properties {soundWaveAssetPtr=/Game/Sounds/Fort_Weapons/pistol/MachinePistol/SoundWaves/Machine_Pistol2_Fire_CloseMech_01}}
 {id=WavePlayer_67, properties {soundWaveAssetPtr=/Game/Sounds/Fort_Weapons/pistol/MachinePistol/SoundWaves/Machine_Pistol2_Fire_CloseMech_02}}
 {id=WavePlayer_68, properties {soundWaveAssetPtr=/Game/Sounds/Fort_Weapons/pistol/MachinePistol/SoundWaves/Machine_Pistol2_Fire_CloseMech_03}}
 {id=WavePlayer_69, properties {soundWaveAssetPtr=/Game/Sounds/Fort_Weapons/pistol/MachinePistol/SoundWaves/Machine_Pistol2_Fire_CloseMech_04}}
 {id=WavePlayer_70, properties {soundWaveAssetPtr=/Game/Sounds/Fort_Weapons/pistol/MachinePistol/SoundWaves/Machine_Pistol2_Fire_CloseMech_05}}
 {id=WavePlayer_71, properties {soundWaveAssetPtr=/Game/Sounds/Fort_Weapons/pistol/MachinePistol/SoundWaves/Machine_Pistol2_Fire_CloseMech_06}}
 {id=WavePlayer_74, properties {soundWaveAssetPtr=/Game/Sounds/Fort_Weapons/pistol/MachinePistol/SoundWaves/Machine_Pistol2_LowAmmo_02}}
 {id=WavePlayer_75, properties {soundWaveAssetPtr=/Game/Sounds/Fort_Weapons/pistol/MachinePistol/SoundWaves/Machine_Pistol2_ShellImpact_01}}
 {id=WavePlayer_76, properties {soundWaveAssetPtr=/Game/Sounds/Fort_Weapons/pistol/MachinePistol/SoundWaves/Machine_Pistol2_ShellImpact_02}}
 {id=WavePlayer_77, properties {soundWaveAssetPtr=/Game/Sounds/Fort_Weapons/pistol/MachinePistol/SoundWaves/Machine_Pistol2_ShellImpact_03}}
 {id=WavePlayer_78, properties {soundWaveAssetPtr=/Game/Sounds/Fort_Weapons/pistol/MachinePistol/SoundWaves/Machine_Pistol2_ShellImpact_04}}
 {id=WavePlayer_79, properties {soundWaveAssetPtr=/Game/Sounds/Fort_Weapons/pistol/MachinePistol/SoundWaves/Machine_Pistol2_ShellImpact_05}}
 {id=WavePlayer_80, properties {soundWaveAssetPtr=/Game/Sounds/Fort_Weapons/pistol/MachinePistol/SoundWaves/Machine_Pistol2_ShellImpact_06}}
]
properties {
 attenuationSettings=&Fort_Weapon_Projectile_FirstPerson_Athena_Att
 bDebug=true
 maxDistance=2097152
 soundClassObject=&Fort_WeaponRanged_Fire_1P_PSM
 volumeMultiplier=1
 concurrencySet [&Weapon_Gunfire_FirstPerson_Conc]
}
