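# Sound cue for Military SMG gunfire (Quartz unvault), layering close, medium and distant shots by distance.
# Each layer picks a random wave through its own attenuation and sound class: close uses 3P fire, while medium and distant share the distant-fire class.
# Close and medium are mixed and pitch-locked by a modulator, then mixed with the modulated distant layer.
# Random nodes are excluded from branch culling, and local and global gunfire concurrency limit playback.
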SoundCue /Game/Sounds/Fort_Weapons/AutoRifle/MilitarySMG/Quartz_Unvault/MilitarySMG_Fire_Cue
root=Mixer_0
nodes [
 {id=Attenuation_1, children [Random_6], properties {attenuationSettings=&Fort_Weapon_Projectile_Close_Athena_Att}}
 {
  id=Attenuation_2
  children [Modulator_10]
  properties {attenuationSettings=&Fort_Weapon_Projectile_Distant_Mono_Athena_Att}
 }
 {
  id=Attenuation_6
  children [Modulator_9]
  properties {attenuationSettings=&Fort_Weapon_Projectile_Medium_Mono_Athena_Att}
 }
 {id=Mixer_0, children [Modulator_6,SoundClass_2], properties {inputVolume [1,0.25]}}
 {id=Mixer_1, children [SoundClass_1,SoundClass_6], properties {inputVolume [0.685,0.25]}}
 {id=Modulator_10, children [Random_8]}
 {id=Modulator_6, children [Mixer_1], properties {pitchMax=1, pitchMin=1}}
 {id=Modulator_9, children [Random_7]}
 {
  id=Random_6
  children [WavePlayer_39, WavePlayer_40, WavePlayer_41, WavePlayer_42, WavePlayer_43, WavePlayer_44, WavePlayer_45, WavePlayer_46]
  properties {bShouldExcludeFromBranchCulling=true, weights [1,1,1,1,1,1,1,1]}
 }
 {
  id=Random_7
  children [WavePlayer_52, WavePlayer_53, WavePlayer_54, WavePlayer_55, WavePlayer_56, WavePlayer_57, WavePlayer_58, WavePlayer_59]
  properties {bShouldExcludeFromBranchCulling=true, weights [1,1,1,1,1,1,1,1]}
 }
 {
  id=Random_8
  children [WavePlayer_62, WavePlayer_63, WavePlayer_64, WavePlayer_65, WavePlayer_66, WavePlayer_67, WavePlayer_68]
  properties {bShouldExcludeFromBranchCulling=true, weights [1,1,1,1,1,1,1]}
 }
 {
  id=SoundClass_1
  children [Attenuation_1]
  properties {soundClassOverride=&Fort_WeaponRanged_Fire_3P_PSM}
 }
 {
  id=SoundClass_2
  children [Attenuation_2]
  properties {soundClassOverride=&Fort_WeaponRanged_Fire_Distant}
 }
 {
  id=SoundClass_6
  children [Attenuation_6]
  properties {soundClassOverride=&Fort_WeaponRanged_Fire_Distant}
 }
 {
  id=WavePlayer_39
  properties {soundWaveAssetPtr=&/Military_SMG_Quartz_Fire_Close_01}
 }
 {
  id=WavePlayer_40
  properties {soundWaveAssetPtr=&/Military_SMG_Quartz_Fire_Close_02}
 }
 {
  id=WavePlayer_41
  properties {soundWaveAssetPtr=&/Military_SMG_Quartz_Fire_Close_03}
 }
 {
  id=WavePlayer_42
  properties {soundWaveAssetPtr=&/Military_SMG_Quartz_Fire_Close_04}
 }
 {
  id=WavePlayer_43
  properties {soundWaveAssetPtr=&/Military_SMG_Quartz_Fire_Close_05}
 }
 {
  id=WavePlayer_44
  properties {soundWaveAssetPtr=&/Military_SMG_Quartz_Fire_Close_06}
 }
 {
  id=WavePlayer_45
  properties {soundWaveAssetPtr=&/Military_SMG_Quartz_Fire_Close_07}
 }
 {
  id=WavePlayer_46
  properties {soundWaveAssetPtr=&/Military_SMG_Quartz_Fire_Close_08}
 }
 {
  id=WavePlayer_52
  properties {soundWaveAssetPtr=&/Military_SMG_Quartz_Fire_Medium_01}
 }
 {
  id=WavePlayer_53
  properties {soundWaveAssetPtr=&/Military_SMG_Quartz_Fire_Medium_02}
 }
 {
  id=WavePlayer_54
  properties {soundWaveAssetPtr=&/Military_SMG_Quartz_Fire_Medium_03}
 }
 {
  id=WavePlayer_55
  properties {soundWaveAssetPtr=&/Military_SMG_Quartz_Fire_Medium_04}
 }
 {
  id=WavePlayer_56
  properties {soundWaveAssetPtr=&/Military_SMG_Quartz_Fire_Medium_05}
 }
 {
  id=WavePlayer_57
  properties {soundWaveAssetPtr=&/Military_SMG_Quartz_Fire_Medium_06}
 }
 {
  id=WavePlayer_58
  properties {soundWaveAssetPtr=&/Military_SMG_Quartz_Fire_Medium_07}
 }
 {
  id=WavePlayer_59
  properties {soundWaveAssetPtr=&/Military_SMG_Quartz_Fire_Medium_08}
 }
 {
  id=WavePlayer_62
  properties {soundWaveAssetPtr=&/Military_SMG_Quartz_Fire_Distant_01}
 }
 {
  id=WavePlayer_63
  properties {soundWaveAssetPtr=&/Military_SMG_Quartz_Fire_Distant_02}
 }
 {
  id=WavePlayer_64
  properties {soundWaveAssetPtr=&/Military_SMG_Quartz_Fire_Distant_03}
 }
 {
  id=WavePlayer_65
  properties {soundWaveAssetPtr=&/Military_SMG_Quartz_Fire_Distant_04}
 }
 {
  id=WavePlayer_66
  properties {soundWaveAssetPtr=&/Military_SMG_Quartz_Fire_Distant_05}
 }
 {
  id=WavePlayer_67
  properties {soundWaveAssetPtr=&/Military_SMG_Quartz_Fire_Distant_06}
 }
 {
  id=WavePlayer_68
  properties {soundWaveAssetPtr=&/Military_SMG_Quartz_Fire_Distant_07}
 }
]
properties {
 bDebug=true
 maxDistance=27500
 volumeMultiplier=1
 concurrencySet [&Weapon_Gunfire_Local_Conc, &Weapon_Gunfire_Global_Conc]
}
